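# Cupid vehicle boost cosmetic Niagara system that streams arrow and heart sprites backward from the car's boost sockets.
# Both emitters partition particles across filtered skeletal mesh sockets, offset them by DTCarOffset, and inherit most of the source's movement.
# Hearts add SubUV animation, shrink out, switch color ranges by GameModeID and select a highlight emissive by ColorVariantID.
# Painted-variant override modules are present but disabled in both emitters.
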NiagaraSystem /VehicleCosmetics/Boosts/Boost_Cupid/FX/NS_BoostTemplate_Cupid_Opt
dependencies {
 usedDataInterfaces [
  &NiagaraDataInterfaceCurve
  &NiagaraDataInterfaceVector2DCurve
  &NiagaraDataInterfaceArrayFloat3
  &NiagaraDataInterfaceSpriteRendererInfo
  &NiagaraDataInterfaceSkeletalMesh
 ]
 usedDynamicInputs []
 usedModules [
  &self:SystemSpawnScript.SystemScriptSource.SystemScriptGraph.NiagaraNodeAssignment_0.SetVariables_#1
  &SystemState
 ]
}
emitters {
 Arrows001 {
  emitterUpdateScript [
   &EmitterState()
   &VC_FilteredSocketCount()
   &SpawnRate(SpawnRate=&RandomRangeFloat(Minimum=15, Maximum=20), Use Spawn Probability=false)
  ]
  particleSpawnScript [
   &VC_PaintedVariantsOverride() {enabled=false}
   &InitializeParticle(
    Color=<rgba 0.4262 0.1998 0.5252 1>
    Lifetime Max=1
    Lifetime Mode=::NewEnumerator1
    Sprite Rotation Angle=&RandomRangeFloat(Minimum=-0.1, Maximum=0.1)
    Sprite Rotation Mode=::NewEnumerator0
    Sprite Size=&Multiply_Vector2D_ByFloat(Float=3, Vector2D=<16 22>)
    Sprite Size Mode=::NewEnumerator4
    UsePositionOffset=false
    Write Material Random=true
   )
   &PartitionParticles(Number Of Partitions=$Emitter.PortCount)
   &SkeletalMeshLocation(
    Bone / Socket Index=$Particles.PartitionParticles.Partition
    Bone / Socket Sampling Mode=::NewEnumerator5
    MaskSpawnGroup=false
    Mesh Sampling Type=::NewEnumerator6
    Sampled Position Offset=&AddVectorToPosition(Position=<-50 0 0>, Vector=$User.DTCarOffset)
    Skeletal Mesh=$System.SkeletalMesh
   )
   &AddVelocity(Velocity=&RandomRangeVector(Minimum=<-800 -128 -64>, Maximum=<-1000 128 170>))
   &ShapeLocation(
    MaskSpawnGroup=false
    Override Seed=false
    Sphere Radius=&RandomRangeFloat(Minimum=&Multiply_Float(A=5, B=2), Maximum=&Multiply_Float(A=10, B=2))
    UseOwnerScale=true
   )
  ]
  particleUpdateScript [
   &ParticleState()
   &InheritSourceMovement(
    Applied Position Scale=&Multiply_VectorByFloat(
     Vector=<1 1 1>
     Float=&FloatFromCurve(FloatCurve=DataInterface() {propertyValues {Curve {keys [
          {interpMode=RCIM_Cubic, leaveTangent=-0.0051, tangentMode=RCTM_Break, time=0, value=1}
          {arriveTangent=-0.247, interpMode=RCIM_Cubic, tangentMode=RCTM_Break, time=1, value=0.9}
         ]}}})
    )
    ApplyForce=false
    ApplyPosition=true
    ApplyPositionFalloff=false
   )
   &ScaleSpriteSize(
    Scale Factor=&Vector2DFromCurve(Vector2Curve=DataInterface() {propertyValues {XCurve {keys [<at 0 1>,<at 1 1>]}, YCurve {keys [<at 0 1.5>,<at 0.1 1>]}}}, ScaleCurve=<1 1>)
    Scale Sprite Size Mode=::NewEnumerator2
   )
   &SolveForcesAndVelocity(Clamp Velocity=false, Limit Acceleration=false)
  ]
  rendererClasses [&NiagaraSpriteRendererProperties]
  renderers [
   {rendererClass=&NiagaraSpriteRendererProperties}
   {rendererClass=&NiagaraSpriteRendererProperties, rendererIndex=1}
   {rendererClass=&NiagaraSpriteRendererProperties, rendererIndex=2}
  ]
 }
 Hearts {
  emitterSpawnScript [/SetVariables(Emitter.Highlightemissive:Vector3f=&SelectVectorFromArray(
     Direct Array Index=$System.ColorVariantID
     Vector Selection Array=DataInterface() {propertyValues {FloatData [
        <30 1 1>
        <0 0 0>
        <200 1 1>
        <60 1 1>
        <60 1 1>
        <60 1 1>
        <60 1 1>
        <60 1 1>
        <60 1 1>
        <60 1 1>
        <60 1 1>
        <60 1 1>
        <60 1 1>
        <30 1 1>
        <30 1 1>
       ]}}
    )) {isSetParameters=true}]
  emitterUpdateScript [
   &EmitterState()
   &VC_FilteredSocketCount()
   &SpawnRate(SpawnRate=&RandomRangeFloat(Minimum=35, Maximum=40), Use Spawn Probability=false)
  ]
  particleSpawnScript [
   &VC_PaintedVariantsOverride() {enabled=false}
   &InitializeParticle(
    Color=&MergeHelper_ColorSwitch(
     0=&RandomRangeLinearColor(Minimum=<rgba 0.9792 0.1077 0.7371 1>, Maximum=<rgba 0.9375 0.082 0.0586 1>, Color Channel Mode=::NewEnumerator0)
     1=&RandomRangeLinearColor(Minimum=<rgba 1 0.0885 0.6962 1>, Maximum=<rgba 1 0.0833 0.0833 1>, Color Channel Mode=::NewEnumerator0)
     2=&RandomRangeLinearColor(Minimum=<rgba 1 0.0885 0.6962 1>, Maximum=<rgba 1 0.0729 0.0729 1>, Color Channel Mode=::NewEnumerator0)
     3=?
     Index=$System.GameModeID
     Num=::NewEnumerator1
    )
    Lifetime Max=1
    Lifetime Min=0.5
    Lifetime Mode=::NewEnumerator1
    Sprite Rotation Angle=&RandomRangeFloat(Minimum=-0.1, Maximum=0.1)
    Sprite Rotation Mode=::NewEnumerator1
    Sprite Size Mode=::NewEnumerator1
    Uniform Sprite Size Max=&Multiply_Float(A=13, B=3)
    Uniform Sprite Size Min=&Multiply_Float(A=22, B=3)
    UsePositionOffset=false
    Write Material Random=true
   )
   &PartitionParticles(Number Of Partitions=$Emitter.PortCount)
   &SkeletalMeshLocation(
    Bone / Socket Index=$Particles.PartitionParticles.Partition
    Bone / Socket Sampling Mode=::NewEnumerator5
    MaskSpawnGroup=false
    Mesh Sampling Type=::NewEnumerator6
    Sampled Position Offset=&AddVectorToPosition(Position=<-10 0 0>, Vector=$User.DTCarOffset)
    Skeletal Mesh=$System.SkeletalMesh
   )
   &AddVelocity(Velocity=&RandomRangeVector(Minimum=<-800 -200 -50>, Maximum=<-1000 200 250>))
   &ShapeLocation(
    MaskSpawnGroup=false
    Override Seed=false
    Sphere Radius=&RandomRangeFloat(Minimum=&Multiply_Float(A=5, B=2), Maximum=&Multiply_Float(A=10, B=2))
    UseOwnerScale=true
   )
  ]
  particleUpdateScript [
   &ParticleState()
   &InheritSourceMovement(
    Applied Position Scale=&Multiply_VectorByFloat(
     Vector=<1 1 1>
     Float=&FloatFromCurve(FloatCurve=DataInterface() {propertyValues {Curve {keys [
          {interpMode=RCIM_Cubic, leaveTangent=-0.0051, tangentMode=RCTM_Break, time=0, value=1}
          {arriveTangent=-0.247, interpMode=RCIM_Cubic, tangentMode=RCTM_Break, time=1, value=0.9}
         ]}}})
    )
    ApplyForce=false
    ApplyPosition=true
    ApplyPositionFalloff=false
   )
   &ScaleSpriteSize(Uniform Curve Sprite Scale=DataInterface() {propertyValues {Curve {keys [
        {tangentMode=RCTM_User, time=0, value=1}
        {interpMode=RCIM_Cubic, tangentMode=RCTM_Break, time=0.7, value=1}
        {interpMode=RCIM_Cubic, tangentMode=RCTM_Break, time=0.95, value=0}
       ]}}})
   &SolveForcesAndVelocity(Clamp Velocity=false, Limit Acceleration=false)
   &SubUVAnimation(
    "Sprite Renderer "=DataInterface() {propertyValues {SpriteRenderer=&self:Hearts_0.NiagaraSpriteRendererProperties_6}}
    SubUV Lookup Index Scale=-1
    UseEndFrame=false
    UseStartFrame=false
   )
   &DynamicMaterialParameters(
    Index 0 Param 1=1.1
    Index 0 Param 2=1
    Index 0 Param 3=1
    Module.Index0Param0Write=true
    Module.Index0Param1Write=true
    Module.Index0Param2Write=true
    Module.Index0Param3Write=true
   )
  ]
  rendererClasses [&NiagaraSpriteRendererProperties]
  renderers [
   {rendererClass=&NiagaraSpriteRendererProperties}
   {rendererClass=&NiagaraSpriteRendererProperties, rendererIndex=1}
  ]
 }
}
summary {
 systemName=NS_BoostTemplate_Cupid_Opt
 emitters [
  {emitterName=Arrows001, rendererClasses [&NiagaraSpriteRendererProperties]}
  {emitterName=Hearts, rendererClasses [&NiagaraSpriteRendererProperties]}
 ]
 userVariables [
  User.bIsFrontendPreview {type=NiagaraBool}
  User.bIsPainted {type=NiagaraBool}
  User.ColorVariantID {type=NiagaraInt32}
  User.DTCarOffset {type=Vector3f, description="", defaultValue {struct=Vector3f, value=<0 0 0>}}
  User.GameModeID {type=NiagaraInt32}
  User.PaintedColor {type=LinearColor, description="", defaultValue {struct=LinearColor, value=<rgba 1 1 1 1>}}
  User.PaintedScalar {type=NiagaraFloat, defaultValue=1, description=""}
  User.PortCount {type=NiagaraInt32, defaultValue=1}
 ]
}
